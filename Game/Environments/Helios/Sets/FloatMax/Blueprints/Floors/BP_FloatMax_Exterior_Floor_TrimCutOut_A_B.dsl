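; Construction script for a FloatMax exterior floor trim cut-out Blueprint in the Helios environment set, which tints its light fixture's material.
; After running the parent script, it checks that the light fixture and its slot 1 material are valid.
; It then creates a dynamic material instance on that slot and sets a hardcoded red on the vector parameter named "Param".
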(fn ConstructionScript
  (bind _lightfixture (Variables|Default|GetLightFixture))
  (bind _returnvalue (Rendering|Material|GetMaterial _lightfixture 1))
  (|CallParentFunction)
  (if (Utilities|IsValid (Variables|Default|GetLightFixture) :!class /Script/Engine.KismetSystemLibrary)
    (if (Utilities|IsValid _returnvalue :!class /Script/Engine.KismetSystemLibrary)
      (bind _returnvalue_2 (Rendering|Material|CreateDynamicMaterialInstance _lightfixture 1 _returnvalue "None" :!class /Script/Engine.PrimitiveComponent))
      (Rendering|Material|SetVectorParameterValue _returnvalue_2 "Param" "(R=1.000000,G=0.000000,B=0.000000,A=1.000000)" :!class /Script/Engine.MaterialInstanceDynamic))))
