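; Blueprint for a FloatMax interior floor piece in the Helios environment set, with a light fixture whose emissive color is set per instance.
; The construction script calls its parent, then checks that the LightFixture and its slot 1 material are valid.
; It creates a dynamic material instance on that slot and writes EmissiveColor into the vector parameter "Param".
; BeginPlay, overlap and Tick are disabled stubs.
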(event EventBeginPlay
  (|CallParentFunction) ; This node is disabled and will not be called.\nDrag off pins to build functionality.
)

(event Collision|EventActorBeginOverlap
  (params (OtherActor :type Actor%20Object%20Reference))
  (|CallParentFunction OtherActor) ; This node is disabled and will not be called.\nDrag off pins to build functionality.
)

(event EventTick
  (params (DeltaSeconds :type Float%20%28single-precision%29))
  (|CallParentFunction DeltaSeconds) ; This node is disabled and will not be called.\nDrag off pins to build functionality.
)

(fn ConstructionScript
  (bind _lightfixture (Variables|Default|GetLightFixture))
  (bind _returnvalue (Rendering|Material|GetMaterial _lightfixture 1))
  (|CallParentFunction)
  (if (Utilities|IsValid (Variables|Default|GetLightFixture) :!class /Script/Engine.KismetSystemLibrary)
    (if (Utilities|IsValid _returnvalue :!class /Script/Engine.KismetSystemLibrary)
      (bind _returnvalue_2 (Rendering|Material|CreateDynamicMaterialInstance _lightfixture 1 _returnvalue "None" :!class /Script/Engine.PrimitiveComponent))
      (Rendering|Material|SetVectorParameterValue _returnvalue_2 "Param" (Variables|Components|GetEmissiveColor) :!class /Script/Engine.MaterialInstanceDynamic))))
